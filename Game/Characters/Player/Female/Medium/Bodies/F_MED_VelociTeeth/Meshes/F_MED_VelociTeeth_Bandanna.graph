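# Physics asset for the bandanna of the medium female player VelociTeeth body, setting up simple capsule physics for its front and back scarf bones.
# A two-link front chain hangs from base_scarfFront and a single back link from base_scarfBack, with tight swing limits and very narrow twist limits.
# A large spine_04 capsule stands in for the torso.
# Only base_scarfBack is constrained to spine_04, with all angular motion locked.
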PhysicsAsset /Game/Characters/Player/Female/Medium/Bodies/F_MED_VelociTeeth/Meshes/F_MED_VelociTeeth_Bandanna
bodies {
 base_scarfBack [{shapeType=Capsule, length=0.5, radius=0.5, shapeName=base_armring_capsule, rotation=<rot 0 -90 0>}]
 base_scarfFront [{shapeType=Capsule, length=0.5, radius=0.5, shapeName=base_armring_capsule, rotation=<rot 0 -90 0>}]
 dyn_scarfBack_01 [{shapeType=Capsule, length=2, radius=1, shapeName=dyn_scarfFront_01_capsule, center=<2 0 0>, rotation=<rot 90 -5 0>}]
 dyn_scarfFront_01 [{shapeType=Capsule, length=2, radius=0.5, shapeName=dyn_scarfFront_01_capsule, center=<2 0 0>, rotation=<rot 90 0 0>}]
 dyn_scarfFront_02 [{shapeType=Capsule, length=2, radius=0.5, shapeName=dyn_scarfFront_01_capsule, center=<2 0 0>, rotation=<rot 90 0 0>}]
 spine_04 [{shapeType=Capsule, length=2, radius=10.5, shapeName=pelvis_capsule, center=<0 2 0>}]
}
constraints [
 {bone1Name=dyn_scarfFront_01, bone2Name=base_scarfFront, swing1LimitDegrees=12.5, swing2LimitDegrees=12.5, twistLimitDegrees=1}
 {bone1Name=dyn_scarfFront_02, bone2Name=dyn_scarfFront_01, swing1LimitDegrees=7.5, swing2LimitDegrees=7.5, twistLimitDegrees=1}
 {bone1Name=dyn_scarfBack_01, bone2Name=base_scarfBack, swing1LimitDegrees=7.5, swing2LimitDegrees=15, twistLimitDegrees=1}
 {
  bone1Name=base_scarfBack
  bone2Name=spine_04
  swing1LimitDegrees=45
  swing1Motion=Locked
  swing2LimitDegrees=45
  swing2Motion=Locked
  twistLimitDegrees=45
  twistMotion=Locked
 }
]
