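# Physics asset for the VelociTeeth body mesh of the medium female player character, covering its egg and pocket-flap attachments.
# Capsule bodies sit on front and back egg bones, a pocket flap, their three base anchor bones, the pelvis and the right thigh.
# Three constraints tie each dyn_ bone to its base bone with tight, per-axis swing and twist limits.
# The larger pelvis and thigh capsules have no constraints.
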PhysicsAsset /Game/Characters/Player/Female/Medium/Bodies/F_MED_VelociTeeth/Meshes/F_MED_VelociTeeth_Physics
bodies {
 base_back_egg [{shapeType=Capsule, length=0.5, radius=0.5, shapeName=base_armring_capsule, rotation=<rot 0 -90 0>}]
 base_front_egg [{shapeType=Capsule, length=0.5, radius=0.5, shapeName=base_armring_capsule, rotation=<rot 0 -90 0>}]
 base_pocket [{shapeType=Capsule, length=0.5, radius=0.5, shapeName=base_pocket_capsule, rotation=<rot 90 0 0>}]
 dyn_back_egg_01 [{shapeType=Capsule, length=3, radius=3.5, shapeName=dyn_back_egg_01_capsule, center=<2.25 -3 0>, rotation=<rot 90 -20 -15>}]
 dyn_front_egg_01 [{shapeType=Capsule, length=3, radius=3.5, shapeName=dyn_front_egg_01_capsule, center=<2.25 -3.5 -1.5>, rotation=<rot 90 -30 -20>}]
 dyn_pocketFlap_01 [{shapeType=Capsule, length=2, radius=1, shapeName=dyn_pocketFlap_01_capsule, center=<2 0 0>, rotation=<rot 90 0 0>}]
 pelvis [{shapeType=Capsule, length=2.5, radius=11.75, shapeName=pelvis_capsule, center=<2 0 0>}]
 thigh_r [{shapeType=Capsule, length=12, radius=7.25, shapeName=base_pocket_capsule, center=<8.5 0 0>, rotation=<rot 100 0 0>}]
}
constraints [
 {bone1Name=dyn_back_egg_01, bone2Name=base_back_egg, swing1LimitDegrees=15, swing2LimitDegrees=35, twistLimitDegrees=5}
 {bone1Name=dyn_front_egg_01, bone2Name=base_front_egg, swing1LimitDegrees=5, swing2LimitDegrees=25, twistLimitDegrees=15}
 {bone1Name=dyn_pocketFlap_01, bone2Name=base_pocket, swing1LimitDegrees=8, swing2LimitDegrees=2.5, twistLimitDegrees=1}
]
