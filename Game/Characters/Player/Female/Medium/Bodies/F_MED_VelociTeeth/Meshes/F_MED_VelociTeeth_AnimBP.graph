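# Secondary-motion AnimBlueprint for the VelociTeeth female medium player body, copying the parent mesh pose into a deform Control Rig.
# Cached poses chain weighted constraints and bone offsets for accessories and scarf, then AnimDynamics bracelets, then two RigidBody simulations (bandanna, body physics asset).
# Each update derives alphas from reference-pose joint deltas to rotate leg spikes with knee bend and shift the back scarf with head-to-spine rotation, plus a gravity override for physics.
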AnimBlueprint /Game/Characters/Player/Female/Medium/Bodies/F_MED_VelociTeeth/Meshes/F_MED_VelociTeeth_AnimBP
target_skeleton=&/F_MED_VelociTeeth_Skeleton
graphs [
 AnimGraph {nodes [
   n1 = RigidBody!(
    ComponentPose=Local To Component(LocalPose=Use cached pose 'Constraints_RigidBody'())
    Alpha
    OverrideWorldGravity:Vector=Get Gravity Override()
    ExternalForce
   ) -> (Pose) {
    bClampLinearTranslationLimitToRefPose=true
    bOverrideWorldGravity=true
    overridePhysicsAsset=&/F_MED_VelociTeeth_Bandanna
    simulationSpace=BaseBoneSpace
    alphaBoolBlend {blendOption=Linear}
    baseBoneRef {boneName=spine_04}
    componentAppliedLinearAccClamp=<3000 3000 3000>
    componentLinearAccScale=<0.3 0.2 0.2>
    componentLinearVelScale=<0.3 0.2 0.2>
    simSpaceSettings {
     maxAngularAcceleration=1000
     maxAngularVelocity=3
     maxLinearAcceleration=700
     maxLinearVelocity=500
     velocityScaleZ=0
     worldAlpha=0.5
     externalLinearDragV=<0.2 0.2>
    }
   }
   n2 = RigidBody(ComponentPose=@n1, OverrideWorldGravity=Get Gravity Override()) {
    bClampLinearTranslationLimitToRefPose=true
    bOverrideWorldGravity=true
    overridePhysicsAsset=&/F_MED_VelociTeeth_Physics
    simulationSpace=BaseBoneSpace
    alphaBoolBlend {blendOption=Linear}
    baseBoneRef {boneName=pelvis}
    componentAppliedLinearAccClamp=<3000 3000 3000>
    componentLinearAccScale=<0.3 0.2 0.2>
    componentLinearVelScale=<0.3 0.2 0.2>
    simSpaceSettings {
     maxAngularAcceleration=1000
     maxAngularVelocity=3
     maxLinearAcceleration=700
     maxLinearVelocity=500
     velocityScaleZ=0
     worldAlpha=0.5
     externalLinearDragV=<0.2 0.2>
    }
   }
   Output Pose(Result=Component To Local(ComponentPose=@n2)) {layerGroup=None}
   n3 = "Constraint - base_scarfFront (PPPP)"!(
    ComponentPose=Local To Component(LocalPose=Use cached pose 'lowBod2UpBod'())
    Alpha=1
    ConstraintWeights_0=0.5505
    ConstraintWeights_1:float=0.4334
    ConstraintWeights_2:float=0.0152
    ConstraintWeights_3:float=0.0009
   ) -> (Pose) {
    alphaBoolBlend {blendOption=Linear}
    boneToModify {boneName=base_scarfFront}
    constraintSetup [{targetBone {boneName=spine_04}}, {targetBone {boneName=spine_05}}, {targetBone {boneName=deform_pec_l}}, {targetBone {boneName=clavicle_l}}]
    constraintWeights [1,1,1,1]
   }
   n4 = "Constraint - base_scarfBack (PPP)"!(ComponentPose=@n3, Alpha=1, ConstraintWeights_0=0.7758, ConstraintWeights_1:float=0.1204, ConstraintWeights_2:float=0.1037) -> (Pose) {
    alphaBoolBlend {blendOption=Linear}
    boneToModify {boneName=base_scarfBack}
    constraintSetup [{targetBone {boneName=spine_05}}, {targetBone {boneName=clavicle_r}}, {targetBone {boneName=clavicle_l}}]
    constraintWeights [1,1,1]
   }
   n5 = "Transform (Modify) Bone - Bone: base_scarfBack"(ComponentPose=@n4, Alpha=1, Translation="0,-0.5,0", Rotation="0,0,0", Scale="1,1,1") {
    rotationSpace=BCS_BoneSpace
    translationMode=BMM_Additive
    translationSpace=BCS_BoneSpace
    alphaBoolBlend {blendOption=Linear}
    boneToModify {boneName=base_scarfBack}
    translation {y=-0.5}
   }
   n6 = "Transform (Modify) Bone - Bone: dyn_scarfBack_02"(ComponentPose=@n5, Alpha=1, Translation="1.8,1.5,0", Rotation="0,0,0", Scale="1,1,1") {
    rotationMode=BMM_Additive
    rotationSpace=BCS_BoneSpace
    scaleMode=BMM_Additive
    scaleSpace=BCS_BoneSpace
    translationMode=BMM_Additive
    translationSpace=BCS_BoneSpace
    alphaBoolBlend {blendOption=Linear}
    boneToModify {boneName=dyn_scarfBack_02}
    translation=<1.8 1.5>
   }
   n7 = "Transform (Modify) Bone - Bone: dyn_scarfBack_02"(ComponentPose=@n6, Alpha=Get bandanna down() -> (bandanna down:float), Translation="4,3,0", Rotation="0,0,0", Scale="1,1,1") {
    rotationSpace=BCS_BoneSpace
    translationMode=BMM_Additive
    translationSpace=BCS_BoneSpace
    alphaBoolBlend {blendOption=Linear}
    boneToModify {boneName=dyn_scarfBack_02}
    translation=<4 3>
   }
   n8 = Anim Dynamics - Bone: dyn_bracelet_01_l!(ComponentPose=@n7, Alpha=1, GravityOverride:Vector=Get Gravity Override(), ExternalForce="0,0,0") -> (Pose) {
    angularDampingOverride=0.96
    bAngularSpring=true
    bGravityOverrideInSimSpace=true
    bLinearSpring=true
    bOverrideAngularDamping=true
    bOverrideLinearDamping=true
    bUseGravityOverride=true
    simulationSpace=BoneRelative
    alphaBoolBlend {blendOption=Linear}
    boundBone {boneName=dyn_bracelet_01_l}
    componentAppliedLinearAccClamp=<10000 10000 10000>
    componentLinearAccScale=<1 1 1>
    componentLinearVelScale=<0.5 0.5 0.5>
    physicsBodyDefinitions [{
      sphereCollisionRadius=0
      boundBone {boneName=dyn_bracelet_01_l}
      boxExtents=<5 18 18>
      constraintSetup {angularTargetAxis=AxisZ, angularLimitsMax=<0 0 -2>, angularLimitsMin=<0 0 1.5>, angularTarget=<0 0 0>}
      localJointOffset=<2 0 -4>
     }]
    relativeSpaceBone {boneName=lowerarm_l}
   }
   n9 = Anim Dynamics - Bone: dyn_frontBracelet_01_r!(ComponentPose=@n8, Alpha=1, GravityOverride:Vector=Get Gravity Override(), ExternalForce="0,0,0") -> (Pose) {
    angularDampingOverride=0.96
    bAngularSpring=true
    bGravityOverrideInSimSpace=true
    bLinearSpring=true
    bOverrideAngularDamping=true
    bOverrideLinearDamping=true
    bUseGravityOverride=true
    simulationSpace=BoneRelative
    alphaBoolBlend {blendOption=Linear}
    boundBone {boneName=dyn_frontBracelet_01_r}
    componentAppliedLinearAccClamp=<10000 10000 10000>
    componentLinearAccScale=<1 1 1>
    componentLinearVelScale=<0.5 0.5 0.5>
    physicsBodyDefinitions [{
      sphereCollisionRadius=0
      boundBone {boneName=dyn_frontBracelet_01_r}
      boxExtents=<5 18 18>
      constraintSetup {angularTargetAxis=AxisZ, angularLimitsMax=<0 0 -1>, angularLimitsMin=<0 0 1.5>, angularTarget=<0 0 0>}
      localJointOffset=<-2 0 4>
     }]
    relativeSpaceBone {boneName=lowerarm_r}
   }
   n10 = Anim Dynamics - Bone: dyn_backBracelet_01_r!(ComponentPose=@n9, Alpha=1, GravityOverride:Vector=Get Gravity Override(), ExternalForce="0,0,0") -> (Pose) {
    angularDampingOverride=0.96
    bAngularSpring=true
    bGravityOverrideInSimSpace=true
    bLinearSpring=true
    bOverrideAngularDamping=true
    bOverrideLinearDamping=true
    bUseGravityOverride=true
    simulationSpace=BoneRelative
    alphaBoolBlend {blendOption=Linear}
    boundBone {boneName=dyn_backBracelet_01_r}
    componentAppliedLinearAccClamp=<10000 10000 10000>
    componentLinearAccScale=<1 1 1>
    componentLinearVelScale=<0.5 0.5 0.5>
    physicsBodyDefinitions [{
      sphereCollisionRadius=0
      boundBone {boneName=dyn_backBracelet_01_r}
      boxExtents=<5 18 18>
      constraintSetup {angularTargetAxis=AxisZ, angularLimitsMax=<0 0 -1>, angularLimitsMin=<0 0 1.5>, angularTarget=<0 0 0>}
      localJointOffset=<-2 0 4>
     }]
    relativeSpaceBone {boneName=lowerarm_r}
   }
   n11 = Anim Dynamics - Bone: dyn_bracelet_02_l!(ComponentPose=@n10, Alpha=1, GravityOverride:Vector=Get Gravity Override(), ExternalForce="0,0,0") -> (Pose) {
    angularDampingOverride=0.96
    bAngularSpring=true
    bGravityOverrideInSimSpace=true
    bLinearSpring=true
    bOverrideAngularDamping=true
    bOverrideLinearDamping=true
    bUseGravityOverride=true
    simulationSpace=BoneRelative
    alphaBoolBlend {blendOption=Linear}
    boundBone {boneName=dyn_bracelet_02_l}
    componentAppliedLinearAccClamp=<10000 10000 10000>
    componentLinearAccScale=<1 1 1>
    componentLinearVelScale=<0.5 0.5 0.5>
    physicsBodyDefinitions [{
      sphereCollisionRadius=0
      boundBone {boneName=dyn_bracelet_02_l}
      boxExtents=<5 18 18>
      constraintSetup {angularTargetAxis=AxisZ, angularLimitsMax=<45 25 6>, angularLimitsMin=<-45 -25 -9>, angularTarget=<0 0 0>}
      localJointOffset=<2 0 -4>
     }]
    relativeSpaceBone {boneName=lowerarm_l}
   }
   n12 = Anim Dynamics - Bone: dyn_frontBracelet_02_r!(ComponentPose=@n11, Alpha=1, GravityOverride:Vector=Get Gravity Override(), ExternalForce="0,0,0") -> (Pose) {
    angularDampingOverride=0.96
    bAngularSpring=true
    bGravityOverrideInSimSpace=true
    bLinearSpring=true
    bOverrideAngularDamping=true
    bOverrideLinearDamping=true
    bUseGravityOverride=true
    simulationSpace=BoneRelative
    alphaBoolBlend {blendOption=Linear}
    boundBone {boneName=dyn_frontBracelet_02_r}
    componentAppliedLinearAccClamp=<10000 10000 10000>
    componentLinearAccScale=<1 1 1>
    componentLinearVelScale=<0.5 0.5 0.5>
    physicsBodyDefinitions [{
      sphereCollisionRadius=0
      boundBone {boneName=dyn_frontBracelet_02_r}
      boxExtents=<5 18 18>
      constraintSetup {angularTargetAxis=AxisZ, twistAxis=AxisY, angularLimitsMax=<0 40 15>, angularLimitsMin=<15 -45 -8>, angularTarget=<0 0 0>}
      localJointOffset=<-2 0 4>
     }]
    relativeSpaceBone {boneName=lowerarm_r}
   }
   n13 = Anim Dynamics - Bone: dyn_backBracelet_02_r!(ComponentPose=@n12, Alpha=1, GravityOverride:Vector=Get Gravity Override(), ExternalForce="0,0,0") -> (Pose) {
    angularDampingOverride=0.96
    bAngularSpring=true
    bGravityOverrideInSimSpace=true
    bLinearSpring=true
    bOverrideAngularDamping=true
    bOverrideLinearDamping=true
    bUseGravityOverride=true
    simulationSpace=BoneRelative
    alphaBoolBlend {blendOption=Linear}
    boundBone {boneName=dyn_backBracelet_02_r}
    componentAppliedLinearAccClamp=<10000 10000 10000>
    componentLinearAccScale=<1 1 1>
    componentLinearVelScale=<0.5 0.5 0.5>
    physicsBodyDefinitions [{
      sphereCollisionRadius=0
      boundBone {boneName=dyn_backBracelet_02_r}
      boxExtents=<5 18 18>
      constraintSetup {angularTargetAxis=AxisZ, angularLimitsMax=<0 40 12.5>, angularLimitsMin=<15 -45 -8>, angularTarget=<0 0 0>}
      localJointOffset=<-2 0 4>
     }]
    relativeSpaceBone {boneName=lowerarm_r}
   }
   Save cached pose 'Constraints_RigidBody'(Pose=Component To Local(ComponentPose=@n13)) {cacheName=Constraints_RigidBody}
   n14 = Get Calf_L() -> (Calf_L:float)
   n15 = "Transform (Modify) Bone - Bone: base_upperspike_L"(
    ComponentPose=Local To Component(LocalPose=Use cached pose 'ControlRig2LegLogic'())
    Alpha=@n14
    Translation="0,0,0"
    Rotation="0,95,0"
    Scale="1,1,1"
   ) {
    rotationMode=BMM_Additive
    rotationSpace=BCS_BoneSpace
    translationMode=BMM_Additive
    translationSpace=BCS_BoneSpace
    alphaBoolBlend {blendOption=Linear}
    alphaScaleBiasClamp {interpSpeedDecreasing=30, interpSpeedIncreasing=0}
    boneToModify {boneName=base_upperspike_L}
    rotation {yaw=95}
   }
   n16 = "Transform (Modify) Bone - Bone: base_lowerspike_L"(ComponentPose=@n15, Alpha=@n14, Translation="0,0,0", Rotation="0,65,0", Scale="1,1,1") {
    rotationMode=BMM_Additive
    rotationSpace=BCS_BoneSpace
    translationMode=BMM_Additive
    translationSpace=BCS_BoneSpace
    alphaBoolBlend {blendOption=Linear}
    alphaScaleBiasClamp {interpSpeedDecreasing=30, interpSpeedIncreasing=0}
    boneToModify {boneName=base_lowerspike_L}
    rotation {yaw=65}
   }
   n17 = Get Calf_R() -> (Calf_R:float)
   n18 = "Transform (Modify) Bone - Bone: base_upperspike_R"(ComponentPose=@n16, Alpha=@n17, Translation="0,0,0", Rotation="0,115,0", Scale="1,1,1") {
    rotationMode=BMM_Additive
    rotationSpace=BCS_BoneSpace
    translationMode=BMM_Additive
    translationSpace=BCS_BoneSpace
    alphaBoolBlend {blendOption=Linear}
    alphaScaleBiasClamp {interpSpeedDecreasing=30, interpSpeedIncreasing=0}
    boneToModify {boneName=base_upperspike_R}
    rotation {yaw=115}
   }
   n19 = "Transform (Modify) Bone - Bone: base_lowerspike_R"(ComponentPose=@n18, Alpha=@n17, Translation="0,0,0", Rotation="0,70,0", Scale="1,1,1") {
    rotationMode=BMM_Additive
    rotationSpace=BCS_BoneSpace
    translationMode=BMM_Additive
    translationSpace=BCS_BoneSpace
    alphaBoolBlend {blendOption=Linear}
    alphaScaleBiasClamp {interpSpeedDecreasing=30, interpSpeedIncreasing=0}
    boneToModify {boneName=base_lowerspike_R}
    rotation {yaw=70}
   }
   n20 = "Constraint - base_pocket (PPPPP)"!(
    ComponentPose=@n19
    Alpha=1
    ConstraintWeights_0=0.6249
    ConstraintWeights_1:float=0.3144
    ConstraintWeights_2:float=0.0524
    ConstraintWeights_3:float=0.0052
    ConstraintWeights_4:float=0.003
   ) -> (Pose) {
    alphaBoolBlend {blendOption=Linear}
    boneToModify {boneName=base_pocket}
    constraintSetup [
     {targetBone {boneName=thigh_twist_01_r}}
     {targetBone {boneName=thigh_r}}
     {targetBone {boneName=pelvis}}
     {targetBone {boneName=deform_groin_r}}
     {targetBone {boneName=deform_glute_r}}
    ]
    constraintWeights [1,1,1,1,1]
   }
   n21 = "Constraint - base_front_egg (PP)"!(ComponentPose=@n20, Alpha=1, ConstraintWeights_0=0.641, ConstraintWeights_1:float=0.359) -> (Pose) {
    alphaBoolBlend {blendOption=Linear}
    boneToModify {boneName=base_front_egg}
    constraintSetup [{targetBone {boneName=spine_01}}, {targetBone {boneName=pelvis}}]
    constraintWeights [1,1]
   }
   n22 = "Constraint - base_back_egg (PP)"!(ComponentPose=@n21, Alpha=1, ConstraintWeights_0=0.628, ConstraintWeights_1:float=0.372) -> (Pose) {
    alphaBoolBlend {blendOption=Linear}
    boneToModify {boneName=base_back_egg}
    constraintSetup [{targetBone {boneName=spine_01}}, {targetBone {boneName=pelvis}}]
    constraintWeights [1,1]
   }
   n23 = "Constraint - base_belt_buckle (PPP)"!(ComponentPose=@n22, Alpha=1, ConstraintWeights_0=0.5331, ConstraintWeights_1:float=0.4666, ConstraintWeights_2:float=0.0003) -> (Pose) {
    alphaBoolBlend {blendOption=Linear}
    boneToModify {boneName=base_belt_buckle}
    constraintSetup [{targetBone {boneName=spine_01}}, {targetBone {boneName=pelvis}}, {targetBone {boneName=deform_groin_l}}]
    constraintWeights [1,1,1]
   }
   Save cached pose 'lowBod2UpBod'(Pose=Component To Local(ComponentPose=@n23)) {cacheName=lowBod2UpBod}
   n24 = Control Rig(Source=Copy Pose From Mesh() {bUseAttachedParent=true}) {
    alphaBoolBlend {blendOption=Linear}
    controlRigAssetReference {blueprintRigClass=&/F_MED_VelociTeeth_DeformRig.F_MED_VelociTeeth_DeformRig_C}
   }
   Save cached pose 'ControlRig2LegLogic'(Pose=@n24) {cacheName=ControlRig2LegLogic}
  ]}
 EventGraph {nodes [
   Event BlueprintUpdateAnimation()
   > n1 = Sequence!(execute) -> (then_0, then_1, then_2:Exec, then_3:Exec)
   > n2 = Gravity Override State Function DEPRECATED(CharacterPartAnimInstance=Self-Reference(), GameTimeInSeconds=GetGameTimeInSeconds(), GravityOverrideParams=Get Gravity Override Params())
   > Set Gravity Override(Gravity Override=@n2)
   Set Skel Mesh(execute=@n1.then_1, Skel Mesh=GetPartSkeletalMeshComponent())
   Set bandanna down(
    execute=@n1.then_3
    bandanna down:float=GetAlphaFromReferenceTransformDelta(SkeletalMeshComponent=Get Skel Mesh(), BoneName=spine_05, BaseName=head, InputRange=(X=0,Y=110), TwistAxis="0,0,1")
   ) -> (Output_Get:float)
   Set Calf_L(
    execute=@n1.then_2
    Calf_L:float=GetAlphaFromReferenceTransformDelta(SkeletalMeshComponent=Get Skel Mesh(), BoneName=thigh_l, BaseName=calf_l, InputRange=(X=-45,Y=-120), TwistAxis="0,0,1")
   ) -> (Output_Get:float)
   > Set Calf_R(Calf_R:float=GetAlphaFromReferenceTransformDelta(SkeletalMeshComponent=Get Skel Mesh(), BoneName=thigh_r, BaseName=calf_r, InputRange=(X=-45,Y=-120), TwistAxis="0,0,1")) -> (Output_Get:float)
  ]}
]
variables [
 bandanna down {type=number}
 Calf_L {type=number}
 Calf_R {type=number}
 Gravity Override {type=Vector}
 Gravity Override Params {type=GravityOverrideParamsStruct}
 Skel Mesh {type=SkeletalMeshComponent}
 Spine Rot {type=number}
]
